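; Blueprint for a posed mannequin actor, apparently a Pliant-specific variant, that manages the playback of its own MannequinAudio component.
; On BeginPlay it stops the audio if hidden, otherwise fades it in with a short linear fade.
; When destroyed it fades the audio out; both events first check that the component is valid.
; The hidden state is read only at BeginPlay, and the construction script is empty.
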(event EventDestroyed
  (bind _mannequinaudio (Variables|Default|GetMannequinAudio)
    (:then
      (Audio|Components|Audio|FadeOut _mannequinaudio 0.25 0.0 "Linear" :!class /Script/Engine.AudioComponent))
    (:else)))

(event EventBeginPlay
  (bind _mannequinaudio (Variables|Default|GetMannequinAudio))
  (Variables|Default|GetMannequinAudio
    (:then
      (if (Variables|Rendering|GetbHidden)
        (Audio|Components|Audio|Stop _mannequinaudio)
        (else
          (Audio|Components|Audio|FadeIn _mannequinaudio 0.25 1.0 0.0 "Linear" :!class /Script/Engine.AudioComponent))))
    (:else)))

(fn ConstructionScript)
